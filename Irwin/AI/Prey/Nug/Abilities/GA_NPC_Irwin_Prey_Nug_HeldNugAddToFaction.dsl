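; Gameplay ability for the Nug (chicken) prey NPC that adds it to a held faction while it is being carried.
; On activation it finds the pawn's FortHeldObjectComponent, ending if none exists, and binds to its held-object events.
; Pickup adds the faction tag; place, throw, drop or destroy removes it.
; Ability end or failed passive activation unbinds all listeners from those events, without removing the tag.
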(event Ability|EventActivateAbility
  (Variables|Default|SetHeldObjectComponent (Actor|GetComponentbyClass (Ability|GetActivatingPawn) "/Script/FortniteGame.FortHeldObjectComponent"))
  (bind _heldobjectcomponent (Variables|Default|GetHeldObjectComponent)
    (:then
      ;[1 1 1] Bind events for chicken being held & not being held
      (EventDispatchers|BindEventtoOnHeldObjectPickedUp _heldobjectcomponent (delegate Custom|Add%20To%20Faction))
      (EventDispatchers|BindEventtoOnHeldObjectPlaced _heldobjectcomponent (delegate Custom|Remove%20From%20Faction))
      (EventDispatchers|BindEventtoOnHeldObjectThrown _heldobjectcomponent (delegate Custom|Remove%20From%20Faction))
      (EventDispatchers|BindEventtoOnHeldObjectDropped _heldobjectcomponent (delegate Custom|Remove%20From%20Faction))
      (EventDispatchers|BindEventtoOnHeldObjectDestroy _heldobjectcomponent (delegate Custom|Remove%20From%20Faction)))
    (:else
      (Ability|EndAbility))))

(event Custom|Add%20To%20Faction
  ;[1 1 1] Add/Remove the Held Faction Tag
  (Affiliation|AddtoFaction (Ability|GetActivatingPawn) (Variables|Default|GetHeld%20Faction%20Tag)))

(event Custom|Remove%20From%20Faction
  ;[1 1 1] Add/Remove the Held Faction Tag
  (Affiliation|RemovefromFaction (Ability|GetActivatingPawn) (Variables|Default|GetHeld%20Faction%20Tag)))

(event Ability|EventFailedToActivatePassiveAbility
  (params (ActorInfo :type Gameplay%20Ability%20Actor%20Info%20Structure))
  (bind _heldobjectcomponent (Variables|Default|GetHeldObjectComponent)
    (:then
      (EventDispatchers|UnbindallEventsfromOnHeldObjectPickedUp _heldobjectcomponent)
      (EventDispatchers|UnbindallEventsfromOnHeldObjectPlaced _heldobjectcomponent)
      (EventDispatchers|UnbindallEventsfromOnHeldObjectDestroy _heldobjectcomponent)
      (EventDispatchers|UnbindallEventsfromOnHeldObjectThrown _heldobjectcomponent)
      (EventDispatchers|UnbindallEventsfromOnHeldObjectDropped _heldobjectcomponent))
    (:else)))

(event Ability|EventOnEndAbility
  (params (bWasCancelled :type Boolean))
  (bind _heldobjectcomponent (Variables|Default|GetHeldObjectComponent)
    (:then
      (EventDispatchers|UnbindallEventsfromOnHeldObjectPickedUp _heldobjectcomponent)
      (EventDispatchers|UnbindallEventsfromOnHeldObjectPlaced _heldobjectcomponent)
      (EventDispatchers|UnbindallEventsfromOnHeldObjectDestroy _heldobjectcomponent)
      (EventDispatchers|UnbindallEventsfromOnHeldObjectThrown _heldobjectcomponent)
      (EventDispatchers|UnbindallEventsfromOnHeldObjectDropped _heldobjectcomponent))
    (:else)))
